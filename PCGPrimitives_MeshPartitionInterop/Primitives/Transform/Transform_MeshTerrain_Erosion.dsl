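; PCG primitive graph that applies an erosion transform, apparently to mesh terrain, with the erosion itself delegated to a core-process subgraph.
; It tags the primary input as "Input:0", then chains template input management, erosion core process and output preparation subgraphs.
; Parameters forward noise-style erosion settings (octaves, scale, gain, lacunarity, strengths) and a mask attribute to the core.
; Each stage's processing graph is exposed as an overridable parameter, including input fallback.
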(graph Transform_MeshTerrain_Erosion
  (params (InputsProcessingGraph :type (SoftObject /Script/PCG.PCGGraphInterface) :default /PCGPrimitives/Primitives/_Template_/Primitives_InputsProcessing.Primitives_InputsProcessing) (InputsFallbackGraph :type (SoftObject /Script/PCG.PCGGraphInterface) :default /PCGPrimitives/Primitives/_Template_/Primitives_InputsFallback.Primitives_InputsFallback) (CoreProcessGraph :type (SoftObject /Script/PCG.PCGGraphInterface) :default /PCGPrimitives_MeshPartitionInterop/Primitives/Transform/Subgraphs/Transform_MeshTerrain_Erosion_CoreProcess.Transform_MeshTerrain_Erosion_CoreProcess) (PrepareOutputsGraph :type (SoftObject /Script/PCG.PCGGraphInterface) :default /PCGPrimitives/Primitives/_Template_/Primitives_PrepareOutputs.Primitives_PrepareOutputs) (ErosionOctaves :type Int :default 5) (ErosionScale :type Double :default 5e+03) (ErosionGain :type Double :default 0.5) (ErosionLacunarity :type Double :default 2.0) (ErosionStrength :type Double :default 5e+02) (ErosionSlopeStength :type Double :default 4.0) (ErosionBranchStrength :type Double :default 3.0) (MaskAttribute :type String))
  (inputs (PrimaryInput :type Any))
  (bind AddTags_1 (Add%20Tags :TagsToAdd "Input:0"))
  ;[0 0 0 0.5 32] Inputs
  (bind Subgraph_0 (Subgraph :Graph /PCGPrimitives/Primitives/_Template_/Subgraphs/Primitives_InputsManagement.Primitives_InputsManagement :InputsFallbackGraph InputsFallbackGraph :InputsProcessingGraph InputsProcessingGraph))
  ;[0 0 0 0.5 32] Core Process
  (bind Subgraph_3 (Subgraph :Graph /PCGPrimitives_MeshPartitionInterop/Primitives/Transform/Subgraphs/Transform_MeshTerrain_Erosion_CoreProcess.Transform_MeshTerrain_Erosion_CoreProcess :ErosionBranchStrength ErosionBranchStrength :ErosionGain ErosionGain :ErosionLacunarity ErosionLacunarity :ErosionOctaves ErosionOctaves :ErosionScale ErosionScale :ErosionSlopeStrength ErosionSlopeStength :ErosionStrength ErosionStrength :MaskAttribute MaskAttribute :SubgraphOverride-param CoreProcessGraph))
  ;[0 0 0 0.5 32] Prepare Output
  (bind Subgraph_4 (Subgraph :Graph /PCGPrimitives/Primitives/_Template_/Primitives_PrepareOutputs.Primitives_PrepareOutputs :SubgraphOverride-param PrepareOutputsGraph))
  (connect DefaultInputNode PrimaryInput AddTags_1 In)
  (connect AddTags_1 Out Subgraph_0 In)
  (connect Subgraph_0 Out Subgraph_3 In)
  (connect Subgraph_3 Out Subgraph_4 In)
  (output Out Subgraph_4))
